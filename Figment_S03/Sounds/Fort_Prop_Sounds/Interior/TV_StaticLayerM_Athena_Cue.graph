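# A SoundCue for a TV static layer ambient prop sound in interior spaces, set up for Athena in Figment Season 03.
# A single wave player plays the TV_StaticLayerM_Figment wave, with raised volume and slightly raised pitch.
# It uses the TV prop attenuation and the ambient tile mix sound class.
# Its concurrency override blocks new instances rather than stopping playing ones.
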SoundCue /Figment_S03/Sounds/Fort_Prop_Sounds/Interior/TV_StaticLayerM_Athena_Cue
root=WavePlayer_2
nodes [{id=WavePlayer_2, properties {soundWaveAssetPtr=&/TV_StaticLayerM_Figment}}]
properties {
 attenuationSettings=/Game/Sounds/Attenuation/Fort_Prop_TV_Athena_Att
 bOverrideConcurrency=true
 maxDistance=500
 pitchMultiplier=1.1
 soundClassObject=/Game/Sounds/FortSoundClassesAndModes/Fort_Ambient_TileMix
 volumeMultiplier=1.5
 attenuationOverrides {lPFRadiusMin=0}
 concurrencyOverrides {resolutionRule=PreventNew}
}
