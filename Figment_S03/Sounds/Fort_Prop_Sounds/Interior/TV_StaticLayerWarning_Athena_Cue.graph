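# Sound cue for an interior TV prop in Athena, named as a static warning layer but playing the TV day loop layer wave.
# A quality-level node plays the full wave at the first level and the "_opt" variant at the other three.
# It uses the shared Athena TV attenuation and ambient sound class, with a concurrency override that prevents new instances once the limit is reached.
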SoundCue /Figment_S03/Sounds/Fort_Prop_Sounds/Interior/TV_StaticLayerWarning_Athena_Cue
root=QualityLevel_0
nodes [
 {id=QualityLevel_0, children [WavePlayer_2,WavePlayer_0,WavePlayer_0,WavePlayer_0]}
 {id=WavePlayer_0, properties {soundWaveAssetPtr=/Game/Sounds/Fort_Prop_Sounds/Interior/TV_Day_Loop_Layer_01_opt}}
 {id=WavePlayer_2, properties {soundWaveAssetPtr=/Game/Sounds/Fort_Prop_Sounds/Interior/TV_Day_Loop_Layer_01}}
]
properties {
 attenuationSettings=/Game/Sounds/Attenuation/Fort_Prop_TV_Athena_Att
 bOverrideConcurrency=true
 maxDistance=500
 soundClassObject=/Game/Sounds/FortSoundClassesAndModes/Fort_Ambient_TileMix
 volumeMultiplier=1.1
 attenuationOverrides {lPFRadiusMin=0}
 concurrencyOverrides {resolutionRule=PreventNew}
}
